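; Gameplay ability, triggered by a gameplay event, that starts hovering on a Fort player pawn by activating the hover Movement Mode Extension.
; ShouldAbilityRespondToEvent rejects activation when the fuel attribute is empty, reporting an out-of-fuel failed action.
; StartHovering rechecks fuel, sends a balloon-removal event and stores the runtime extension.
; The ability ends on the Hover.End event or failed startup, and EndAbility cleanup ends the extension.
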(event Ability|EventActivateAbilityFromEvent
  (params (EventData :type Gameplay%20Event%20Data%20Structure%20%28by%20ref%29))
  ;[0 1 0] Activated from event
  (|CallParentFunction EventData)
  (bind _outfortplayerpawn (CallFunction|GetAvatarasValidFortPlayerPawn)
    (:ValidFortPlayerPawn
      (Variables|Default|SetPlayer _outfortplayerpawn)
      (CallFunction|SetupBindsandWatchEvents)
      (CallFunction|StartHovering))
    (:AvatarCastFailed
      (Ability|EndAbility))))

(event Custom|SetupBindsAndWatchEvents
  ;[1 0.286 0.1076] End ability when MME ends
  (Ability|Tasks|WaitGameplayEvent "(TagName=\"Event.Athena.Hover.End\")"
    (:EventReceived
      (Ability|EndAbility))))

(event Custom|StartHovering
  ;[0.2 0.2 1] Start Hovering
  (bind _player (Variables|Default|GetPlayer)
    (:then
      (if (CallFunction|HasFuel _player)
        (Ability|SendGameplayEventtoActor _player "(TagName=\"Event.Athena.Balloons.RemoveAll\")")
        (bind _returnvalue (Class|FortMovementModeExtBlueprintLibrary|ActivateRegisteredMovementModeExtension (Variables|MovementModeExtension|GetHoverMME) (Variables|Default|GetPlayer)))
        (bind _asmme_runtime_hover (Utilities|Casting|CastToMME_Runtime_Hover _returnvalue))
        (Variables|MovementModeExtension|SetHoverMME_Runtime _asmme_runtime_hover)
        (else
          _)))
    (:else
      _))
  ;;[1 0 0 0.5] End the ability in fail cases
  (Ability|EndAbility))

(event Ability|EventOnEndAbility
  (params (bWasCancelled :type Boolean))
  ;[1 0.25 0] EndAbility Cleanup
  (|CallParentFunction bWasCancelled)
  (bind _hovermme_runtime (Variables|MovementModeExtension|GetHoverMME_Runtime)
    (:then
      (bind _player (Variables|Default|GetPlayer)
        (:then
          (Class|FortMovementModeExtBlueprintLibrary|EndMovementModeExtensiononActor _hovermme_runtime _player))
        (:else)))
    (:else)))

(fn HasFuel
  (params (Actor :type Actor%20Object%20Reference))
  (bind (_bsuccessfullyfoundattribute _returnvalue) (Ability|Attribute|GetFloatAttribute Actor (Variables|Default|GetFuelAttribute)))
  (Utilities|IsValid Actor
    (:Is%20Valid
      (return (> _returnvalue 0)))
    (:Is%20Not%20Valid
      (return false))))

(fn ShouldAbilityRespondToEvent
  (params (ActorInfo :type Gameplay%20Ability%20Actor%20Info%20Structure) (Payload :type Gameplay%20Event%20Data%20Structure))
  (bind (_owneractor _avataractor) (Utilities|Struct|BreakGameplayAbilityActorInfo ActorInfo))
  (bind _asfort_player_pawn (Utilities|Casting|CastToFortPlayerPawn _avataractor)
    (:then
      (if (not (CallFunction|HasFuel _avataractor))
        (FailedAction|UnabletoPerformAction _asfort_player_pawn false (GameplayTags|MakeGameplayTagContainerfromTag (Variables|Default|GetOutOfFuelFailReason)))
        (return false)
        (else
          _)))
    (:CastFailed
      _))
  (return true))
