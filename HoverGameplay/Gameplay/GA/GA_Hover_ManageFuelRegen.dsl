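; Gameplay ability that manages hover fuel regeneration for the owning Fortnite pawn.
; When the hover-in-use tag is added, it cancels the pending timer, removes regen by asset tag and optionally applies a block-until-land effect via hotfix.
; When the tag is removed with fuel below 100, a hotfixable cooldown timer applies the regen GameplayEffect.
; An attribute threshold task removes regen at 100, as does ability end.
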(event Ability|EventActivateAbility
  (bind _grantedhoverinusetag (Variables|Default|GetGrantedHoverInUseTag))
  (bind _outfortplayerpawn (CallFunction|GetAvatarasValidFortPlayerPawn)
    (:ValidFortPlayerPawn
      (bind _output_get (Variables|Default|SetOwnerPawn _outfortplayerpawn))
      (Utilities|FlowControl|Sequence
        (:then_0
          ;[0.5 1 0.5] Watching hover usage to stop regen and then reapply after a cooldown once they are not used.
          (Ability|Tasks|WaitGameplayTagAdd _grantedhoverinusetag _output_get
            (:Added
              (Utilities|Time|ClearandInvalidateTimerbyHandle (Variables|Default|GetTimer))
              (CallFunction|StopFuelRegen)
              (if (ScalableFloat|GetValueasBool (Variables|Hotfix|GetSF_BlockTilLand))
                (Ability|ApplyGameplayEffectToOwner (Variables|Default|GetGE_BlockRegenTilLand))))))
        (:then_1
          (Ability|Tasks|WaitGameplayTagRemove _grantedhoverinusetag
            (:Removed
              (bind (_valid _fuel) (CallFunction|GetFuelLevel))
              (if (and _valid (< _fuel 1e+02))
                (bind _returnvalue (Utilities|Time|SetTimerbyEvent (EventDispatchers|CreateEvent :!event Custom|StartFuelRegen) (ScalableFloat|GetValueatLevel (Variables|Hotfix|GetSF_FuelRegenCooldown))))
                (Variables|Default|SetTimer _returnvalue)))))
        (:then_2
          ;[1 0.3352 0.1135] Stop fuel regen at 100
          (bind (_asynctaskproxy _bmatchescomparison) (Ability|Tasks|WaitForAttributeChangeThreshold (Variables|Default|GetFuelAttribute) "GreaterThanOrEqualTo" 1e+02 false _output_get)
            (:OnChange
              (if _bmatchescomparison
                (CallFunction|StopFuelRegen)))))))))

(event Custom|StartFuelRegen
  ;[0.2 0.2 1] Start and Stop fuel regen
  (Ability|ApplyGameplayEffectToOwner (Variables|Default|GetGE_HoverFuelRegen)))

(event Custom|StopFuelRegen
  ;[0.2 0.2 1] Start and Stop fuel regen
  (Ability|RemoveGameplayEffectFromOwnerWithAssetTags (Variables|Default|GetFuelRegenTag)))

(event Ability|EventOnEndAbility
  (params (bWasCancelled :type Boolean))
  ;[0.2 0.2 1] Start and Stop fuel regen
  (Ability|RemoveGameplayEffectFromOwnerWithAssetTags (Variables|Default|GetFuelRegenTag)))

(fn GetFuelLevel
  (bind _ownerpawn (Variables|Default|GetOwnerPawn)
    (:then
      (return (Ability|Attribute|GetFloatAttribute _ownerpawn (Variables|Default|GetFuelAttribute)) (.ReturnValue (Ability|Attribute|GetFloatAttribute _ownerpawn (Variables|Default|GetFuelAttribute)))))
    (:else
      (return false 0.0))))
